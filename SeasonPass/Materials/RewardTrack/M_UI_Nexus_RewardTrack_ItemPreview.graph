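# Translucent UI material for the Nexus reward track item preview, drawing an item texture with hover, hold and archived states.
# Animation scalars drive centered UV scaling, with position and per-instance pan offsets.
# The texture can switch to a flipbook or be recolored as channel-packed masks.
# A rounded-corner mask, radial vignette and archived linear gradient supply alpha masks, while HoldEffect and ColorWash parameters brighten and tint color.
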Material /SeasonPass/Materials/RewardTrack/M_UI_Nexus_RewardTrack_ItemPreview
graph {
 properties {
  bUsedWithStaticMesh=false
  blendMode=BLEND_Translucent
  materialDomain=MD_UI
  pixelDepthOffsetMode=PDOM_Legacy
  usedShadingModels=Default Lit
 }
 nodes [
  n1 = NamedRerouteUsage()
  n2 = Power(Base=@n1) {constExponent=0.5}
  n3 = LinearInterpolate(A=@n1, B=Saturate(None=Add(A=@n2) {constB=0.65}), Alpha=ScalarParameter() {parameterName=HoldEffect}) {constA=0.5, constAlpha=1}
  LinearInterpolate(
   A=@n3
   B=Add(A=@n3, B=VectorParameter() {parameterName=ColorWashColor, defaultValue=<rgba 0.9249 0.7827 1 1>}.RGB)
   Alpha=ScalarParameter() {parameterName=ColorWash}
  )
  MaterialFunctionCall() {materialFunction=&DebugTimeSine}
  n4 = ScalarParameter() {defaultValue=20, group=Layout, parameterName=TopLeftCornerRadius, sortPriority=4}
  n5 = ScalarParameter() {defaultValue=20, group=Layout, parameterName=BottomLeftCornerRadius, sortPriority=7}
  n6 = ScalarParameter() {defaultValue=20, group=Layout, parameterName=TopRightCornerRadius, sortPriority=5}
  n7 = ScalarParameter() {defaultValue=20, group=Layout, parameterName=BottomRightCornerRadius, sortPriority=6}
  n8 = ScalarParameter() {
   defaultValue=1
   desc="If enabled reads corner radius values in pixels.\r\nbool"
   group=Layout
   parameterName=UseValueInPixels
   sliderMax=1
   sortPriority=1
  }
  NamedRerouteDeclaration(None=MaterialFunctionCall(
    Corner Radius=MaterialFunctionCall(R=@n4, G=@n5, B=@n6, A=@n7) {materialFunction=&AppendMany}.RGBA
    Radius in Pixels=@n8
   ) {materialFunction=&MF_UI_CoreUI_RoundedMask}.Shape Mask) {name=ShapeMask, nodeColor=<rgba 0 0 0 1>}
  n9 = TextureObjectParameter() {group=Item, parameterName=Item, sortPriority=1, texture=/FortUILibrary/Systems/Items/Textures/T_UI_SoldierCommando_F}
  n10 = MaterialFunctionCall(
   "Animation  Phase (0-1)"=Multiply(A=Time(), B=ScalarParameter() {defaultValue=1, group=FlipBook, parameterName=Speed})
   Number of Rows=ScalarParameter() {defaultValue=0.5, group=FlipBook, parameterName=Rows}
   Number of Columns=ScalarParameter() {defaultValue=4, group=FlipBook, parameterName=Columns}
   Texture=@n9
   UVs=NamedRerouteUsage()
  ) {materialFunction=&FlipBook}
  n11 = ScalarParameter() {group=FlipBook, parameterName=Use FLipbook}
  n12 = LinearInterpolate(A=NamedRerouteUsage(), B=@n10.Result, Alpha=@n11)
  NamedRerouteDeclaration(None=ComponentMask(None=@n12) {b=true}) {name=BMask, nodeColor=<rgba 0 0 0 1>}
  NamedRerouteDeclaration(None=ComponentMask(None=@n12) {g=true}) {name=GMask, nodeColor=<rgba 0 0 0 1>}
  NamedRerouteDeclaration(None=ComponentMask(None=@n12) {r=true}) {name=RMask, nodeColor=<rgba 0 0 0 1>}
  n13 = TextureSample(UVs=NamedRerouteUsage(), Tex=@n9) {samplerSource=SSM_Clamp_WorldGroupSettings}
  NamedRerouteDeclaration(None=@n13.RGB) {name=TextureRGB, nodeColor=<rgba 0 1 0.7891 1>}
  NamedRerouteDeclaration(None=@n13.A) {name=TextureA, nodeColor=<rgba 1 0.4688 0 1>}
  n14 = ScalarParameter() {
   desc="(0 or 1) If enabled switch to work with channel-packed texture."
   group=ChannelPacked
   parameterName=UseChannelPacked
   sliderMax=1
   sortPriority=1
  }
  NamedRerouteDeclaration(None=@n14) {name=UseChannelPacked, nodeColor=<rgba 0 0.0156 1 1>}
  n15 = VectorParameter() {
   desc=Color that applies to channel-packed texture
   group=ChannelPacked
   parameterName=DynamicColor
   sortPriority=2
   defaultValue=<rgba 1 0 1 1>
  }
  n16 = LinearInterpolate(A=LinearInterpolate(B=@n15.RGB, Alpha=NamedRerouteUsage()), Alpha=NamedRerouteUsage())
  n17 = NamedRerouteUsage()
  NamedRerouteDeclaration(None=LinearInterpolate(A=@n12, B=LinearInterpolate(A=@n16, Alpha=@n17) {constB=0}, Alpha=NamedRerouteUsage())) {name=Item, nodeColor=<rgba 0 0 0 1>}
  n18 = MaterialFunctionCall!(UVs=NamedRerouteUsage(), Alpha=LinearInterpolate(A=NamedRerouteUsage(), B=@n10.Alpha, Alpha=@n11)) {materialFunction=/FortUILibrary/MaterialFunctions/Basics/MF_UI_ClipEdgePixel}
  n19 = ScalarParameter() {group=Item, parameterName=RemoveItem}
  n20 = Add(A=NamedRerouteUsage(), B=Add(A=NamedRerouteUsage(), B=NamedRerouteUsage()))
  NamedRerouteDeclaration(None=LinearInterpolate(A=LinearInterpolate(A=@n18.Result, Alpha=@n19) {constB=0}, B=Saturate(None=@n20), Alpha=NamedRerouteUsage())) {name=Item Alpha, nodeColor=<rgba 0 0 0 1>}
  NamedRerouteDeclaration(None=MaterialFunctionCall() {materialFunction=&MF_UI_UVScaling}.ScaleY) {name=ScaleY, nodeColor=<rgba 0 0 0 1>}
  n21 = Add(
   A=NamedRerouteUsage()
   B=AppendVector(A=ScalarParameter() {parameterName=CharacterPositionX}, B=ScalarParameter() {parameterName=CharacterPositionY})
  )
  n22 = ScalarParameter() {defaultValue=1, desc=Start value for scale animation., group=Item, parameterName=DefaultScale, sortPriority=2}
  n23 = ScalarParameter() {defaultValue=1.1, desc=Start value for scale animation., group=Item, parameterName=HoverScale, sortPriority=2}
  n24 = ScalarParameter() {parameterName=ScaleAnimation}
  n25 = ScalarParameter() {defaultValue=1.25, desc=End value for scale animation., group=Item, parameterName=HoldScale, sortPriority=3}
  n26 = MaterialFunctionCall() {materialFunction=&GetUserInterfaceUV}
  n27 = Saturate(None=Divide(A=ComponentMask(None=@n26.Pixel Size) {g=true}, B=ComponentMask(None=@n26.Pixel Size) {r=true}))
  n28 = ScalarParameter() {desc="(0 - 1) Animate scale transition 'ScaleFrom' - 'ScaleTo'", group=Item, parameterName=HoldAnimation, sliderMax=1, sortPriority=4}
  n29 = Multiply(
   A=LinearInterpolate(A=LinearInterpolate(A=@n22, B=@n23, Alpha=@n24), B=Multiply(A=@n25, B=@n27), Alpha=@n28)
   B=ScalarParameter() {defaultValue=1, group=0 - Custom MI, parameterName=MI - Scale}
  )
  n30 = ScalarParameter() {group=0 - Custom MI, parameterName=MI - Pan X, sortPriority=45}
  n31 = ScalarParameter() {group=0 - Custom MI, parameterName=MI - Pan Y, sortPriority=50}
  NamedRerouteDeclaration(None=Subtract(
    A=MaterialFunctionCall(UVs=@n21, Texture Scale=@n29) {materialFunction=&ScaleUVsByCenter}.UVs
    B=MaterialFunctionCall(X=@n30, Y=@n31) {materialFunction=&MakeFloat2}.Result
   )) {name=UVPanFprCustomMI, nodeColor=<rgba 1 0 0.6562 1>}
  n32 = AppendVector(
   A=ScalarParameter() {defaultValue=0.5, group=Vignette, parameterName=VignetteCenter_X, sortPriority=10}
   B=ScalarParameter() {group=Vignette, parameterName=VignetteCenter_Y, sortPriority=11}
  )
  n33 = ScalarParameter() {defaultValue=1.4, group=Vignette, parameterName=VignetteRadius, sortPriority=12}
  n34 = ScalarParameter() {defaultValue=3.5, group=Vignette, parameterName=VignetteDensity, sortPriority=13}
  n35 = Multiply(
   A=OneMinus(None=MaterialFunctionCall(CenterPosition=@n32, Radius=@n33, Density=@n34) {materialFunction=&RadialGradientExponential}.RadialGradientExponential)
   B=ScalarParameter() {defaultValue=1, group=Vignette, parameterName=VignetteStrength, sortPriority=6}
  )
  n36 = ScalarParameter() {
   defaultValue=1
   desc="(0 or 1) If enabled applies vignette effect on the item texture."
   group=Vignette
   parameterName=UseVignette
   sliderMax=1
   sortPriority=0
  }
  NamedRerouteDeclaration(None=Multiply(A=Saturate(None=@n35), B=@n36)) {name=Vignette Alpha, nodeColor=<rgba 0 0 0 1>}
  n37 = OneMinus(None=MaterialFunctionCall(
    GradientStart=ScalarParameter() {group=Archived, parameterName=ArchivedGradientStart, sortPriority=10}
    GradientEnd=ScalarParameter() {defaultValue=2, group=Archived, parameterName=ArchivedGradientEnd, sortPriority=11}
    RotationAngle=ScalarParameter() {defaultValue=0.75, group=Archived, parameterName=ArchivedGradientAngle, sortPriority=12}
   ) {materialFunction=&MF_UI_LinearGradient}.Result)
  n38 = ScalarParameter() {
   defaultValue=0.95
   desc=Multiplies Archived Colors alpha. A quick way to set desired threshold in applying the archived effect.
   group=Archived
   parameterName=ArchivedAlphaMultiplier
   sliderMax=1
   sortPriority=3
  }
  n39 = ScalarParameter() {
   desc="(0 - 1) Animate item Archiving effect.\r\n0: Item not archived.\r\n1: Item archived."
   group=Archived
   parameterName=ArchivedAnimation
   sliderMax=1
   sortPriority=0
  }
  NamedRerouteDeclaration(None=LinearInterpolate(B=Multiply(A=@n37, B=@n38), Alpha=@n39) {constA=1}) {name=Archived Alpha, nodeColor=<rgba 0 0 0 1>}
  n40 = Multiply(A=OneMinus(None=NamedRerouteUsage()), B=NamedRerouteUsage())
  n41 = NamedRerouteUsage()
  Reroute(None=Multiply(A=Multiply(A=@n40, B=@n41), B=NamedRerouteUsage()))
 ]
}
