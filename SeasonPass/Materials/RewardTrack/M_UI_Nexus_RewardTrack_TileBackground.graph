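# Translucent UI material that draws the tile background for the season pass reward track.
# It feeds the blurred Slate post buffer into CoreUI backplate B, brightened and clamped, and backplate C, blending between them with the IsDark parameter.
# A CoreUI rounded mask, with per-corner radii optionally in pixels, provides the tile's shape through a named reroute.
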Material /SeasonPass/Materials/RewardTrack/M_UI_Nexus_RewardTrack_TileBackground
graph {
 properties {
  bAutomaticallySetUsageInEditor=false
  bUsedWithStaticMesh=false
  blendMode=BLEND_Translucent
  materialDomain=MD_UI
  usedShadingModels=Default Lit
 }
 nodes [
  n1 = MaterialFunctionCall!("UVs (Optional)") {materialFunction=/Engine/Functions/UserInterface/GetSlatePost0}
  n2 = MaterialFunctionCall!(Post Buffer Blur=@n1.LinearRGB) {materialFunction=/FortUILibrary/MaterialFunctions/CoreUI/MF_UI_CoreUI_BackplateB}
  LinearInterpolate(
   A=Saturate(None=Divide(A=@n2.Result) {constB=0.8})
   B=MaterialFunctionCall!(Post Buffer Blur=@n1.LinearRGB) {materialFunction=/FortUILibrary/MaterialFunctions/CoreUI/MF_UI_CoreUI_BackplateC}.Result
   Alpha=ScalarParameter() {parameterName=IsDark}
  )
  n3 = ScalarParameter() {defaultValue=20, group=Layout, parameterName=TopLeftCornerRadius, sortPriority=4}
  n4 = ScalarParameter() {defaultValue=20, group=Layout, parameterName=BottomLeftCornerRadius, sortPriority=7}
  n5 = ScalarParameter() {defaultValue=20, group=Layout, parameterName=TopRightCornerRadius, sortPriority=5}
  n6 = ScalarParameter() {defaultValue=20, group=Layout, parameterName=BottomRightCornerRadius, sortPriority=6}
  n7 = ScalarParameter() {
   defaultValue=1
   desc="If enabled reads corner radius values in pixels.\r\nbool"
   group=Layout
   parameterName=UseValueInPixels
   sliderMax=1
   sortPriority=1
  }
  NamedRerouteDeclaration(None=MaterialFunctionCall(
    Corner Radius=MaterialFunctionCall(R=@n3, G=@n4, B=@n5, A=@n6) {materialFunction=&AppendMany}.RGBA
    Radius in Pixels=@n7
   ) {materialFunction=&MF_UI_CoreUI_RoundedMask}.Shape Mask) {name=Shape Mask, nodeColor=<rgba 0 0 0 1>}
  NamedRerouteUsage()
 ]
}
