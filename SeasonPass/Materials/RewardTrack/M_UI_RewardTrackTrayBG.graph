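# Translucent UI material for the season pass reward track tray background: a skewed backplate with an animated tech grid.
# Two offset box SDFs form left and right fins, unioned, antialiased and combined with a vertical step fill into the backplate mask.
# The StormTech grid function, with viewport-scaled tiling and time-driven seed, is faded by a radial gradient.
# A color lerp blends backplate and grid colors, with parameters exposed for shape and pattern.
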Material /SeasonPass/Materials/RewardTrack/M_UI_RewardTrackTrayBG
graph {
 expressions {
  MaterialExpressionAdd_0 {
   class=MaterialExpressionAdd
   inputs [
    A {from=MaterialExpressionMaterialFunctionCall_1, output=Skewed UVs}
    B {from=MaterialExpressionConvert_5}
   ]
  }
  MaterialExpressionAdd_1 {
   class=MaterialExpressionAdd
   inputs [
    A {from=MaterialExpressionReroute_0}
    B {from=MaterialExpressionConvert_6}
   ]
  }
  MaterialExpressionComponentMask_0 {class=MaterialExpressionComponentMask, inputs [None {from=MaterialExpressionNamedRerouteUsage_2}], properties {g=true}}
  MaterialExpressionConstant2Vector_1 {class=MaterialExpressionConstant2Vector, properties {g=0.85, r=0.925}}
  MaterialExpressionConstant2Vector_3 {class=MaterialExpressionConstant2Vector, properties {g=0.2, r=0.1}}
  MaterialExpressionConstant2Vector_4 {class=MaterialExpressionConstant2Vector, properties {g=0.5, r=0.75}}
  MaterialExpressionConstant2Vector_5 {class=MaterialExpressionConstant2Vector, properties {g=0.3}}
  MaterialExpressionConstant_0 {class=MaterialExpressionConstant}
  MaterialExpressionConstant_1 {class=MaterialExpressionConstant, properties {r=1}}
  MaterialExpressionConstant_2 {class=MaterialExpressionConstant, properties {r=1}}
  MaterialExpressionConstant_23 {class=MaterialExpressionConstant, properties {r=0.05}}
  MaterialExpressionConstant_3 {class=MaterialExpressionConstant, properties {r=0.025}}
  MaterialExpressionConstant_6 {class=MaterialExpressionConstant}
  MaterialExpressionConvert_0 {
   class=MaterialExpressionConvert
   inputs [
    None {from=""}
    None {from=MaterialExpressionScalarParameter_0}
   ]
   properties {
    nodeName=Make Vector2
    convertInputs [
     {type=Scalar, expressionInput {expression=null, inputName=None, mask=0, maskA=0, maskB=0, maskG=0, maskR=0, outputIndex=0}}
     {
      type=Scalar
      expressionInput {
       expression=&self:MaterialExpressionScalarParameter_0
       inputName=None
       mask=0
       maskA=0
       maskB=0
       maskG=0
       maskR=0
       outputIndex=0
      }
     }
    ]
    convertOutputs [{type=Vector2}]
   }
  }
  MaterialExpressionConvert_5 {
   class=MaterialExpressionConvert
   inputs [
    None {from=MaterialExpressionScalarParameter_4}
    None {from=MaterialExpressionScalarParameter_5}
   ]
   properties {
    nodeName=Make Vector2
    convertInputs [
     {
      type=Scalar
      expressionInput {
       expression=&self:MaterialExpressionScalarParameter_4
       inputName=None
       mask=0
       maskA=0
       maskB=0
       maskG=0
       maskR=0
       outputIndex=0
      }
     }
     {
      type=Scalar
      expressionInput {
       expression=&self:MaterialExpressionScalarParameter_5
       inputName=None
       mask=0
       maskA=0
       maskB=0
       maskG=0
       maskR=0
       outputIndex=0
      }
     }
    ]
    convertOutputs [{type=Vector2}]
   }
  }
  MaterialExpressionConvert_6 {
   class=MaterialExpressionConvert
   inputs [
    None {from=MaterialExpressionScalarParameter_6}
    None {from=MaterialExpressionScalarParameter_20}
   ]
   properties {
    nodeName=Make Vector2
    convertInputs [
     {
      type=Scalar
      expressionInput {
       expression=&self:MaterialExpressionScalarParameter_6
       inputName=None
       mask=0
       maskA=0
       maskB=0
       maskG=0
       maskR=0
       outputIndex=0
      }
     }
     {
      type=Scalar
      expressionInput {
       expression=&self:MaterialExpressionScalarParameter_20
       inputName=None
       mask=0
       maskA=0
       maskB=0
       maskG=0
       maskR=0
       outputIndex=0
      }
     }
    ]
    convertOutputs [{type=Vector2}]
   }
  }
  MaterialExpressionConvert_7 {
   class=MaterialExpressionConvert
   inputs [
    None {from=MaterialExpressionScalarParameter_23}
    None {from=MaterialExpressionScalarParameter_24}
   ]
   properties {
    nodeName=Make Vector2
    convertInputs [
     {
      type=Scalar
      expressionInput {
       expression=&self:MaterialExpressionScalarParameter_23
       inputName=None
       mask=0
       maskA=0
       maskB=0
       maskG=0
       maskR=0
       outputIndex=0
      }
     }
     {
      type=Scalar
      expressionInput {
       expression=&self:MaterialExpressionScalarParameter_24
       inputName=None
       mask=0
       maskA=0
       maskB=0
       maskG=0
       maskR=0
       outputIndex=0
      }
     }
    ]
    convertOutputs [{type=Vector2}]
   }
  }
  MaterialExpressionDivide_0 {
   class=MaterialExpressionDivide
   inputs [
    A {from=MaterialExpressionScalarParameter_1}
    B {from=MaterialExpressionMaterialFunctionCall_0, output=Viewport Scale Factor}
   ]
  }
  MaterialExpressionLinearInterpolate_19 {
   class=MaterialExpressionLinearInterpolate
   inputs [
    A {from=MaterialExpressionVectorParameter_0, output=RGB}
    B {from=MaterialExpressionVectorParameter_9, output=RGB}
    Alpha {from=MaterialExpressionMultiply_1}
   ]
  }
  MaterialExpressionMaterialFunctionCall_0 {
   class=MaterialExpressionMaterialFunctionCall
   inputs []
   properties {materialFunction=/FortUILibrary/MaterialFunctions/Basics/MF_UI_ViewportScale}
  }
  MaterialExpressionMaterialFunctionCall_1 {
   class=MaterialExpressionMaterialFunctionCall
   inputs [
    UV {from=MaterialExpressionSubtract_0}
    Skew Angle {from=MaterialExpressionConvert_7}
   ]
   properties {materialFunction=/FortUILibrary/MaterialFunctions/Basics/MF_UI_Skew}
  }
  MaterialExpressionMaterialFunctionCall_2 {
   class=MaterialExpressionMaterialFunctionCall
   inputs [
    DistanceFieldA {from=MaterialExpressionMaterialFunctionCall_6, output=SDF}
    DistanceFieldB {from=MaterialExpressionMaterialFunctionCall_7, output=SDF}
    BlendAmount {from=""}
   ]
   properties {materialFunction=/FortUILibrary/MaterialFunctions/Shapes/Modifiers/MF_UI_SDF_Union}
  }
  MaterialExpressionMaterialFunctionCall_3 {
   class=MaterialExpressionMaterialFunctionCall
   inputs [
    Screen UV {from=""}
    Tiling {from=MaterialExpressionReroute_4}
    Rotation {from=""}
    Main Min/Max Opacity {from=MaterialExpressionReroute_5}
    Main Min/Max Scale {from=MaterialExpressionReroute_6}
    Main Min/Max Perspective {from=MaterialExpressionReroute_10}
    Main Outline Thickness {from=""}
    Main Speed {from=MaterialExpressionReroute_12}
    Main Reduction {from=MaterialExpressionReroute_14}
    Sub Min/Max Opacity {from=MaterialExpressionReroute_16}
    Sub Min/Max Scale {from=MaterialExpressionReroute_20}
    Sub Min/Max Perspective {from=MaterialExpressionReroute_22}
    Sub Outline Thickness {from=""}
    Sub Speed {from=MaterialExpressionReroute_18}
    Sub Reduction {from=MaterialExpressionReroute_24}
    Random Seed {from=MaterialExpressionMultiply_6}
    Time {from=""}
   ]
   properties {materialFunction=/FortUILibrary/MaterialFunctions/BR/MF_UI_BR_StormTech}
  }
  MaterialExpressionMaterialFunctionCall_4 {
   class=MaterialExpressionMaterialFunctionCall
   inputs [
    UVs {from=MaterialExpressionNamedRerouteUsage_4}
    CenterPosition {from=""}
    Radius {from=MaterialExpressionScalarParameter_2}
    Density {from=""}
    Invert Density {from=""}
   ]
   properties {materialFunction=&RadialGradientExponential}
  }
  MaterialExpressionMaterialFunctionCall_5 {
   class=MaterialExpressionMaterialFunctionCall
   inputs [
    DistanceField {from=MaterialExpressionMaterialFunctionCall_2, output=Exact}
    Threshold {from=""}
    Smoothing {from=""}
   ]
   properties {materialFunction=&MF_UI_AntiAliasing}
  }
  MaterialExpressionMaterialFunctionCall_6 {
   class=MaterialExpressionMaterialFunctionCall
   inputs [
    UVs {from=MaterialExpressionAdd_0}
    Size {from=MaterialExpressionConstant_1}
    CornerRoundness {from=""}
    KeepAspectRatio {from=""}
    PixelSize {from=""}
   ]
   properties {materialFunction=/FortUILibrary/MaterialFunctions/Shapes/SDFs/MF_SDF_Box}
  }
  MaterialExpressionMaterialFunctionCall_7 {
   class=MaterialExpressionMaterialFunctionCall
   inputs [
    UVs {from=MaterialExpressionAdd_1}
    Size {from=MaterialExpressionConstant_2}
    CornerRoundness {from=""}
    KeepAspectRatio {from=""}
    PixelSize {from=""}
   ]
   properties {materialFunction=&MF_SDF_Box}
  }
  MaterialExpressionMax_0 {
   class=MaterialExpressionMax
   inputs [
    A {from=MaterialExpressionMaterialFunctionCall_5, output=Inner}
    B {from=MaterialExpressionStep_1}
   ]
  }
  MaterialExpressionMin_1 {
   class=MaterialExpressionMin
   inputs [
    A {from=MaterialExpressionScalarParameter_22, output=""}
    B {from=MaterialExpressionScalarParameter_21, output=""}
   ]
   properties {}
  }
  MaterialExpressionMultiply_0 {
   class=MaterialExpressionMultiply
   inputs [
    A {from=MaterialExpressionMaterialFunctionCall_3, output=Grid}
    B {from=MaterialExpressionMaterialFunctionCall_4, output=RadialGradientExponential}
   ]
  }
  MaterialExpressionMultiply_1 {
   class=MaterialExpressionMultiply
   inputs [
    A {from=MaterialExpressionNamedRerouteUsage_41}
    B {from=MaterialExpressionVectorParameter_9, output=A}
   ]
  }
  MaterialExpressionMultiply_6 {
   class=MaterialExpressionMultiply
   inputs [
    A {from=MaterialExpressionTime_2}
    B {from=MaterialExpressionScalarParameter_7}
   ]
  }
  MaterialExpressionNamedRerouteDeclaration_0 {
   class=MaterialExpressionNamedRerouteDeclaration
   inputs [None {from=MaterialExpressionMultiply_0}]
   properties {name=Grid Mask, nodeColor=<rgba 1 0 0.2578 1>}
  }
  MaterialExpressionNamedRerouteDeclaration_1 {
   class=MaterialExpressionNamedRerouteDeclaration
   inputs [None {from=MaterialExpressionTextureCoordinate_3}]
   properties {name=UV, nodeColor=<rgba 0.2656 1 0 1>}
  }
  MaterialExpressionNamedRerouteDeclaration_2 {
   class=MaterialExpressionNamedRerouteDeclaration
   inputs [None {from=MaterialExpressionMax_0}]
   properties {name=Backplate Mask, nodeColor=<rgba 0 1 0.4609 1>}
  }
  MaterialExpressionNamedRerouteUsage_1 {class=MaterialExpressionNamedRerouteUsage}
  MaterialExpressionNamedRerouteUsage_2 {class=MaterialExpressionNamedRerouteUsage}
  MaterialExpressionNamedRerouteUsage_3 {class=MaterialExpressionNamedRerouteUsage}
  MaterialExpressionNamedRerouteUsage_4 {class=MaterialExpressionNamedRerouteUsage}
  MaterialExpressionNamedRerouteUsage_41 {class=MaterialExpressionNamedRerouteUsage}
  MaterialExpressionReroute_0 {class=MaterialExpressionReroute, inputs [None {from=MaterialExpressionMaterialFunctionCall_1, output=Skewed UVs}]}
  MaterialExpressionReroute_10 {class=MaterialExpressionReroute, inputs [None {from=MaterialExpressionReroute_9}]}
  MaterialExpressionReroute_11 {class=MaterialExpressionReroute, inputs [None {from=MaterialExpressionConstant_23}]}
  MaterialExpressionReroute_12 {class=MaterialExpressionReroute, inputs [None {from=MaterialExpressionReroute_11}]}
  MaterialExpressionReroute_13 {class=MaterialExpressionReroute, inputs [None {from=MaterialExpressionScalarParameter_9}]}
  MaterialExpressionReroute_14 {class=MaterialExpressionReroute, inputs [None {from=MaterialExpressionReroute_13}]}
  MaterialExpressionReroute_15 {class=MaterialExpressionReroute, inputs [None {from=MaterialExpressionConstant2Vector_5}]}
  MaterialExpressionReroute_16 {class=MaterialExpressionReroute, inputs [None {from=MaterialExpressionReroute_15}]}
  MaterialExpressionReroute_17 {class=MaterialExpressionReroute, inputs [None {from=MaterialExpressionConstant_3}]}
  MaterialExpressionReroute_18 {class=MaterialExpressionReroute, inputs [None {from=MaterialExpressionReroute_17}]}
  MaterialExpressionReroute_19 {class=MaterialExpressionReroute, inputs [None {from=MaterialExpressionConstant2Vector_4}]}
  MaterialExpressionReroute_20 {class=MaterialExpressionReroute, inputs [None {from=MaterialExpressionReroute_19}]}
  MaterialExpressionReroute_21 {class=MaterialExpressionReroute, inputs [None {from=MaterialExpressionConstant_0}]}
  MaterialExpressionReroute_22 {class=MaterialExpressionReroute, inputs [None {from=MaterialExpressionReroute_21}]}
  MaterialExpressionReroute_23 {class=MaterialExpressionReroute, inputs [None {from=MaterialExpressionScalarParameter_8}]}
  MaterialExpressionReroute_24 {class=MaterialExpressionReroute, inputs [None {from=MaterialExpressionReroute_23}]}
  MaterialExpressionReroute_4 {class=MaterialExpressionReroute, inputs [None {from=MaterialExpressionDivide_0}]}
  MaterialExpressionReroute_5 {class=MaterialExpressionReroute, inputs [None {from=MaterialExpressionReroute_7}]}
  MaterialExpressionReroute_6 {class=MaterialExpressionReroute, inputs [None {from=MaterialExpressionReroute_8}]}
  MaterialExpressionReroute_7 {class=MaterialExpressionReroute, inputs [None {from=MaterialExpressionConstant2Vector_3}]}
  MaterialExpressionReroute_8 {class=MaterialExpressionReroute, inputs [None {from=MaterialExpressionConstant2Vector_1}]}
  MaterialExpressionReroute_9 {class=MaterialExpressionReroute, inputs [None {from=MaterialExpressionConstant_6}]}
  MaterialExpressionScalarParameter_0 {class=MaterialExpressionScalarParameter, properties {defaultValue=0.2, group=Backplate, parameterName=Backplate Y Offset}}
  MaterialExpressionScalarParameter_1 {class=MaterialExpressionScalarParameter, properties {defaultValue=16, group=Pattern, parameterName=Tiling}}
  MaterialExpressionScalarParameter_2 {class=MaterialExpressionScalarParameter, properties {defaultValue=0.5, group=Pattern, parameterName=Grid Mask Radius}}
  MaterialExpressionScalarParameter_20 {class=MaterialExpressionScalarParameter, properties {defaultValue=-0.1, group=Backplate, parameterName=Right Fin Y Offset}}
  MaterialExpressionScalarParameter_21 {class=MaterialExpressionScalarParameter, properties {defaultValue=-0.1, group=Backplate, parameterName=Right Fin Y Offset}}
  MaterialExpressionScalarParameter_22 {class=MaterialExpressionScalarParameter, properties {group=Backplate, parameterName=Left Fin Y Offset}}
  MaterialExpressionScalarParameter_23 {class=MaterialExpressionScalarParameter, properties {defaultValue=20, group=Backplate, parameterName=Backplate X Skew}}
  MaterialExpressionScalarParameter_24 {class=MaterialExpressionScalarParameter, properties {defaultValue=4, group=Backplate, parameterName=Backplate Y Skew}}
  MaterialExpressionScalarParameter_3 {class=MaterialExpressionScalarParameter, properties {defaultValue=0.2, group=Backplate, parameterName=Backplate Y Offset}}
  MaterialExpressionScalarParameter_4 {class=MaterialExpressionScalarParameter, properties {defaultValue=0.4, group=Backplate, parameterName=Left Fin X Offset}}
  MaterialExpressionScalarParameter_5 {class=MaterialExpressionScalarParameter, properties {group=Backplate, parameterName=Left Fin Y Offset}}
  MaterialExpressionScalarParameter_6 {class=MaterialExpressionScalarParameter, properties {defaultValue=-0.5, group=Backplate, parameterName=Right Fin X Offset}}
  MaterialExpressionScalarParameter_7 {class=MaterialExpressionScalarParameter, properties {defaultValue=0.25, group=Pattern, parameterName=Grid Speed}}
  MaterialExpressionScalarParameter_8 {class=MaterialExpressionScalarParameter, properties {defaultValue=1, group=Pattern, parameterName=Sub Reduction}}
  MaterialExpressionScalarParameter_9 {class=MaterialExpressionScalarParameter, properties {group=Pattern, parameterName=Main Reduction}}
  MaterialExpressionStep_1 {
   class=MaterialExpressionStep
   inputs [
    Y {from=MaterialExpressionSubtract_1, output=""}
    X {from=MaterialExpressionComponentMask_0, output=""}
   ]
   properties {constX=0.2, constY=0.2}
  }
  MaterialExpressionSubtract_0 {
   class=MaterialExpressionSubtract
   inputs [
    A {from=MaterialExpressionNamedRerouteUsage_1}
    B {from=MaterialExpressionConvert_0}
   ]
  }
  MaterialExpressionSubtract_1 {
   class=MaterialExpressionSubtract
   inputs [
    A {from=MaterialExpressionScalarParameter_3}
    B {from=MaterialExpressionMin_1}
   ]
  }
  MaterialExpressionTextureCoordinate_3 {class=MaterialExpressionTextureCoordinate}
  MaterialExpressionTime_2 {class=MaterialExpressionTime}
  MaterialExpressionVectorParameter_0 {
   class=MaterialExpressionVectorParameter
   properties {group=Backplate, parameterName=Backplate Color, defaultValue=<rgba 0.0055 0.0021 0.0521 1>}
  }
  MaterialExpressionVectorParameter_9 {class=MaterialExpressionVectorParameter, properties {group=Pattern, parameterName=Grid Color, defaultValue=<rgba 0.1217 0 1 1>}}
 }
 properties {
  bAutomaticallySetUsageInEditor=false
  bUsedWithStaticMesh=false
  blendMode=BLEND_Translucent
  materialDomain=MD_UI
  usedShadingModels=Default Lit
 }
}
