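# MF_UI_BattlePassTrayMask is a UI material function that supplies a mask for the battle pass tray.
# It outputs a lerp from full white to the Shape Mask of the BPReboot landing page overlay function.
# The PM_Transition_ShowMask scalar in the Post Match group drives the blend, so the mask can be faded in or bypassed, apparently for post-match transitions.
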MaterialFunction /SeasonPass/MaterialFunctions/MF_UI_BattlePassTrayMask
graph {
 properties {previewBlendMode=BLEND_Translucent, previewMaterialDomain=MD_UI}
 nodes [FunctionOutput(None=LinearInterpolate(
    B=MaterialFunctionCall() {materialFunction=&/MF_UI_BPReboot_LandingPage_Overlay}.Shape Mask
    Alpha=ScalarParameter() {group=Post Match, parameterName=PM_Transition_ShowMask, sliderMax=1}
   ) {constA=1})]
}
